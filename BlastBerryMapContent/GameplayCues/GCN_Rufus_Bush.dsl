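; Looping gameplay cue actor that dresses a player in a bush, attached to the pawn's pelvis.
; On start it re-shows recycled components, hides the outer bush for the local wearer, fades the owner's bush opacity toward a curve-table value and binds footstep rustle sounds (local or third-person).
; Tick sets leaf-animation material parameters in three pawn-speed tiers.
; Removal hides the outer bush, plays destruction effects and unbinds footsteps.
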(event EventOnLoopingStart(Cascade)
  (params (MyTarget :type Actor%20Object%20Reference) (Parameters :type Gameplay%20Cue%20Parameters%20Structure%20%28by%20ref%29) (ParticleComponents :type Array%20of%20Cascade%20Particle%20System%20Component%20Object%20References) (AudioComponents :type Array%20of%20Audio%20Component%20Object%20References))
  (bind _bushmesh (Variables|Default|GetBushMesh))
  ;[0 1 0.02905] On Looping Start
  (Variables|Default|SetPlayerPawn MyTarget)
  (Actor|Tick|SetActorTickEnabled true)
  ;;[0.6048 0 1] Manually turn on the elements we want visible.
  ;; These components can be turned off by GCN_RezOut when the player dies and because this asset can be recycled we need manually turn them back on.
  (Development|SetHiddeninGame _bushmesh false false)
  (Rendering|SetVisibility _bushmesh true false)
  ;;[0 0.1282 1] Hide outer bush mesh for local (equipping) player
  (CallFunction|CheckLocalPlayerandHideBush)
  (Transformation|SetRelativeRotation (Variables|Default|GetArrow) "0.000000,90.000000,-90.000000" false false)
  (bind _returnvalue (Rendering|Material|CreateDynamicMaterialInstance _bushmesh 0 0 "None" :!class /Script/Engine.PrimitiveComponent))
  (Variables|Default|SetMID_Bush _returnvalue)
  (bind _returnvalue_2 (Rendering|Material|CreateDynamicMaterialInstance (Variables|Default|GetOwningPlayer_BushMesh) 0 0 "None" :!class /Script/Engine.PrimitiveComponent)) ; Opaque material
  (Variables|Default|SetMID_OwningClientBush _returnvalue_2)
  (bind _outxy (Fortnite|EvaluateCurveTableRow (Variables|Default|GetClientBushOpacity)))
  (Variables|Default|SetTargetOpacity _outxy)
  (Utilities|FlowControl|Sequence
    (:then_0
      (Variables|Default|SetIsActive true)
      (bind _asfort_player_pawn_athena (Utilities|Casting|CastToFortPlayerPawnAthena (Variables|Default|GetPlayerPawn)))
      (EventDispatchers|BindEventtoOnFootstepEvent _asfort_player_pawn_athena (EventDispatchers|CreateEvent :!event Custom|On%20Player%20Step))
      (Transformation|AttachActorToComponent (Variables|/Script/Engine.Character|GetMesh _asfort_player_pawn_athena) "pelvis" "SnapToTarget" "SnapToTarget" "KeepWorld"))
    (:then_1
      (bind (_direction _opacityfadeout) (Timeline|OpacityFade PlayFromStart)
        (:Update
          (Rendering|Material|SetScalarParameterValue (Variables|Default|GetMID_OwningClientBush) "Opacity" (+ _opacityfadeout (Variables|Default|GetTargetOpacity)) :!class /Script/Engine.MaterialInstanceDynamic))))))

(event EventOnRemoval(Cascade)
  (params (MyTarget :type Actor%20Object%20Reference) (Parameters :type Gameplay%20Cue%20Parameters%20Structure%20%28by%20ref%29) (ParticleComponents :type Array%20of%20Cascade%20Particle%20System%20Component%20Object%20References) (AudioComponents :type Array%20of%20Audio%20Component%20Object%20References) (BurstCameraShakeInstance :type Camera%20Shake%20Base%20Object%20Reference) (BurstDecalInstance :type Decal%20Actor%20Object%20Reference))
  (bind _returnvalue (Transformation|GetActorTransform))
  (bind _location (.location _returnvalue))
  (Actor|Tick|SetActorTickEnabled false)
  (Rendering|SetVisibility (Variables|Default|GetBushMesh) false false)
  (Effects|SpawnEmitteratLocation 0 (Variables|Default|GetBushDestructionVFX) _location (Math|Rotator|MakeRotator 0.0 0.0 (.yaw (.rotation _returnvalue))) (.scale _returnvalue))
  (Audio|PlaySoundatLocation 0 (Variables|Default|GetBushDestructionSFX) _location)
  (bind _playerpawn (Variables|Default|GetPlayerPawn)
    (:then
      (bind _asfort_player_pawn_athena (Utilities|Casting|CastToFortPlayerPawnAthena _playerpawn))
      (EventDispatchers|UnbindEventfromOnFootstepEvent _asfort_player_pawn_athena (EventDispatchers|CreateEvent :!event Custom|On%20Player%20Step))
      (Timeline|OpacityFade Stop))
    (:else)))

(event Custom|On%20Player%20Step
  (params (FootIndex :type Integer))
  (bind _playerpawn (Variables|Default|GetPlayerPawn))
  (bind _asfort_player_pawn_athena (Utilities|Casting|CastToFortPlayerPawnAthena _playerpawn))
  (Audio|PlaySoundatLocation (Variables|Default|GetPlayerPawn) (select (Pawn|IsLocallyControlled _asfort_player_pawn_athena) "/Game/Athena/Sounds/Bush/Athena_Character_Bush_Footstep_Cue.Athena_Character_Bush_Footstep_Cue" "/Game/Athena/Sounds/Bush/Athena_Character_Bush_Footstep_3P_Cue.Athena_Character_Bush_Footstep_3P_Cue") (Transformation|GetActorLocation _playerpawn)))

(event EventTick
  (params (DeltaSeconds :type Float%20%28single-precision%29))
  (bind _playerpawn (Variables|Default|GetPlayerPawn))
  (bind _returnvalue (Transformation|GetVelocity _playerpawn))
  (bind _returnvalue_2 (Math|Vector|VectorLength _returnvalue))
  (bind _returnvalue_3 (/ _returnvalue_2 500))
  (bind _returnvalue_4 (Math|Float|Clamp(Float) _returnvalue_3))
  (bind _mid_bush (Variables|Default|GetMID_Bush))
  (bind _mid_bush_2 (Variables|Default|GetMID_Bush))
  (bind _mid_bush_3 (Variables|Default|GetMID_Bush))
  (if (and (Variables|Default|GetIsActive) (Utilities|IsValid _playerpawn :!class /Script/Engine.KismetSystemLibrary))
    (if (> _returnvalue_4 0.8)
      (Rendering|Material|SetScalarParameterValue _mid_bush_2 "LeafAnimationSpeed" 0.5 :!class /Script/Engine.MaterialInstanceDynamic)
      (Rendering|Material|SetScalarParameterValue _mid_bush_2 "LeafAnimationAmount" 0.3 :!class /Script/Engine.MaterialInstanceDynamic)
      (elif (> _returnvalue_4 0.2)
        (Rendering|Material|SetScalarParameterValue _mid_bush "LeafAnimationSpeed" 0.25 :!class /Script/Engine.MaterialInstanceDynamic)
        (Rendering|Material|SetScalarParameterValue _mid_bush "LeafAnimationAmount" 0.15 :!class /Script/Engine.MaterialInstanceDynamic)
        (else
          (Rendering|Material|SetScalarParameterValue _mid_bush_3 "LeafAnimationSpeed" 0.08 :!class /Script/Engine.MaterialInstanceDynamic)
          (Rendering|Material|SetScalarParameterValue _mid_bush_3 "LeafAnimationAmount" 0.05 :!class /Script/Engine.MaterialInstanceDynamic))))))

(fn CheckLocalPlayerandHideBush
  (if (Ability|GameplayCue|IsInstigatorLocallyControlledPlayer (Ability|GameplayCue|MakeGameplayCueParameters :NormalizedMagnitude 0.0 :RawMagnitude 0.0 :Location "0,0,0" :Normal "0,0,0" :Instigator (Variables|Default|GetPlayerPawn) :GameplayEffectLevel 1 :AbilityLevel 1 :bReplicateLocationWhenUsingMinimalRepProxy false))
    (Rendering|SetVisibility (Variables|Default|GetBushMesh) false false)))

(fn ConstructionScript)
